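# Physics asset for the F_MED_Critter_Streak female player body cosmetic, apparently for secondary motion on its hood and vest.
# Capsule bodies cover the dynamic cloth bones, plus larger spine and thigh shapes that likely act as colliders.
# Hood and upper vest pieces constrain to spine_05, lower and back vest pieces to spine_01.
# Several chains use small base bodies with free angular joints, followed by swing-limited segments, mostly with zero twist limits.
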PhysicsAsset /BRCosmetics/Characters/Player/Female/Medium/Bodies/F_MED_Critter_Streak/Meshes/F_MED_Critter_Streak_Physics
bodies {
 dyn_hood_1 [{shapeType=Capsule, length=7.25, radius=2.5, shapeName=dyn_hood_1_capsule, center=<1.0534 -0.608 0.0001>}]
 dyn_hood_base [{shapeType=Capsule, length=0.505, radius=0.505, shapeName=dyn_hood_base_capsule, rotation=<rot 0 -90 0>}]
 dyn_vestTp_l_1 [{shapeType=Capsule, length=2.5, radius=1.75, shapeName=dyn_vestTp_l_1_capsule, center=<0.5 0 0>}]
 dyn_vestTp_r_1 [{shapeType=Capsule, length=2.5, radius=1.75, shapeName=dyn_vestTp_r_1_capsule, center=<-0.5 0 0>}]
 dyn_vestTp_r_2 [{shapeType=Capsule, length=1.25, radius=1.75, shapeName=dyn_vestTp_r_2_capsule, center=<0.2 0 0>, rotation=<rot -90 0 0>}]
 dyn_vest_back_1_l [{shapeType=Capsule, length=9, radius=1.5, shapeName=dyn_vest_back_1_l_capsule, center=<5.0549 1.455 -0.0001>, rotation=<rot 90 0 0>}]
 dyn_vest_back_1_r [{shapeType=Capsule, length=7.75, radius=1, shapeName=dyn_vest_back_1_r_capsule, center=<4.4451 0 -0.0001>, rotation=<rot 90 0 0>}]
 dyn_vest_back_base_c [{shapeType=Capsule, length=0, radius=0.5, shapeName=dyn_vest_back_base_c_capsule}]
 dyn_vest_back_base_l [{shapeType=Capsule, length=0, radius=0.5, shapeName=dyn_vest_back_base_l_capsule}]
 dyn_vest_back_base_r [{shapeType=Capsule, length=0, radius=0.5, shapeName=dyn_vest_back_base_r_capsule}]
 dyn_vest_back_c [{shapeType=Capsule, length=8.75, radius=1, shapeName=dyn_vest_back_c_capsule, center=<4.7833 1.2523 -0.0001>, rotation=<rot 90 0 0>}]
 dyn_vest_l_1 [{shapeType=Capsule, length=10.75, radius=2, shapeName=dyn_vest_l_1_capsule, center=<8 -2.0001 1.5418>, rotation=<rot 90 0 0>}]
 dyn_vest_l_2 [{shapeType=Capsule, length=1.7508, radius=0.9107, shapeName=dyn_vest_l_2_capsule, center=<1.5802 0 0>, rotation=<rot 90 0 0>}]
 dyn_vest_l_base [{shapeType=Capsule, length=0.505, radius=0.505, shapeName=dyn_vest_l_base_capsule, rotation=<rot 0 -90 0>}]
 dyn_vest_r_1 [{shapeType=Capsule, length=13, radius=2, shapeName=dyn_vest_r_1_capsule, center=<-6.4 0.8107 -0.3012>, rotation=<rot 90 0 0>}]
 spine_01 [{shapeType=Capsule, length=5, radius=9.75, shapeName=spine_01_capsule, center=<0 1.9987 0>}]
 spine_05 [{shapeType=Capsule, length=14.75, radius=8.5, shapeName=spine_05_capsule, center=<2.3343 -0.7846 0>}]
 thigh_l [{shapeType=Capsule, length=14.0917, radius=7.265, shapeName=thigh_l_capsule, center=<-9.1438 0.0002 0>, rotation=<rot 90 -17.354 -12.8957>}]
 thigh_r [{
   shapeType=Capsule
   length=28.1755
   radius=7.8485
   shapeName=thigh_r_capsule
   center=<6.1554 -0.2573 -0.9151>
   rotation=<rot 85.0374 177.291 180>
  }]
}
constraints [
 {
  bone1Name=spine_05
  bone2Name=spine_01
  swing1LimitDegrees=45
  swing1Motion=Free
  swing2LimitDegrees=45
  swing2Motion=Free
  twistLimitDegrees=45
  twistMotion=Free
 }
 {bone1Name=dyn_vestTp_r_2, bone2Name=spine_05, swing1LimitDegrees=10, swing2LimitDegrees=20, twistLimitDegrees=0}
 {
  bone1Name=dyn_vest_l_base
  bone2Name=spine_01
  swing1LimitDegrees=45
  swing1Motion=Free
  swing2LimitDegrees=45
  swing2Motion=Free
  twistLimitDegrees=45
  twistMotion=Free
 }
 {
  bone1Name=dyn_hood_base
  bone2Name=spine_05
  swing1LimitDegrees=45
  swing1Motion=Free
  swing2LimitDegrees=45
  swing2Motion=Free
  twistLimitDegrees=45
  twistMotion=Free
 }
 {bone1Name=dyn_hood_1, bone2Name=dyn_hood_base, swing1LimitDegrees=15, swing2LimitDegrees=10, twistLimitDegrees=0}
 {bone1Name=dyn_vestTp_l_1, bone2Name=spine_05, swing1LimitDegrees=10, swing2LimitDegrees=0, twistLimitDegrees=0}
 {bone1Name=dyn_vestTp_r_1, bone2Name=spine_05, swing1LimitDegrees=10, swing2LimitDegrees=0, twistLimitDegrees=0}
 {
  bone1Name=dyn_vest_back_base_l
  bone2Name=spine_01
  swing1LimitDegrees=45
  swing1Motion=Free
  swing2LimitDegrees=45
  swing2Motion=Free
  twistLimitDegrees=45
  twistMotion=Free
 }
 {bone1Name=dyn_vest_back_1_l, bone2Name=dyn_vest_back_base_l, swing1LimitDegrees=10, swing2LimitDegrees=0, twistLimitDegrees=0}
 {
  bone1Name=dyn_vest_back_base_c
  bone2Name=spine_01
  swing1LimitDegrees=45
  swing1Motion=Free
  swing2LimitDegrees=45
  swing2Motion=Free
  twistLimitDegrees=45
  twistMotion=Free
 }
 {bone1Name=dyn_vest_back_c, bone2Name=dyn_vest_back_base_c, swing1LimitDegrees=10, swing2LimitDegrees=0, twistLimitDegrees=0}
 {
  bone1Name=dyn_vest_back_base_r
  bone2Name=spine_01
  swing1LimitDegrees=45
  swing1Motion=Free
  swing2LimitDegrees=45
  swing2Motion=Free
  twistLimitDegrees=45
  twistMotion=Free
 }
 {bone1Name=dyn_vest_back_1_r, bone2Name=dyn_vest_back_base_r, swing1LimitDegrees=10, swing2LimitDegrees=0, twistLimitDegrees=0}
 {bone1Name=dyn_vest_l_1, bone2Name=dyn_vest_l_base, swing1LimitDegrees=45, swing2LimitDegrees=20, twistLimitDegrees=30}
 {bone1Name=dyn_vest_r_1, bone2Name=spine_01, swing1LimitDegrees=80, swing2LimitDegrees=30, twistLimitDegrees=0}
 {bone1Name=dyn_vest_l_2, bone2Name=dyn_vest_l_1, swing1LimitDegrees=65.5714, swing2LimitDegrees=45, twistLimitDegrees=0}
]
