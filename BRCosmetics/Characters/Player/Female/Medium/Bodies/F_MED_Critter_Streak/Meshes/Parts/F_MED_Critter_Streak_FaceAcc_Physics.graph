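# Physics asset for the FaceAcc part of the female medium Critter Streak player cosmetic, setting up physics for its dynamic hair bones.
# Capsule bodies form side, back and front hair chains on both sides plus a center back chain, each rooted at the head with limited swing and twist.
# Capsules on head, neck, upper spine and clavicles approximate the upper body, linked by angularly free constraints.
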PhysicsAsset /BRCosmetics/Characters/Player/Female/Medium/Bodies/F_MED_Critter_Streak/Meshes/Parts/F_MED_Critter_Streak_FaceAcc_Physics
bodies {
 clavicle_l [{
   shapeType=Capsule
   length=21.505
   radius=6.255
   shapeName=clavicle_l_capsule
   center=<-1.7448 0.7826 -0.8975>
   rotation=<rot -1.8348 -94.2255 86.6436>
  }]
 clavicle_r [{
   shapeType=Capsule
   length=21.505
   radius=6.255
   shapeName=clavicle_r_capsule
   center=<1.6005 0.043 0.9216>
   rotation=<rot -1.8348 -94.2254 86.6436>
  }]
 dyn_hair_1_l [{shapeType=Capsule, length=4.5, radius=2, shapeName=dyn_hair_1_l_capsule, center=<3.0271 -0.763 0>, rotation=<rot 90 0 -179.9998>}]
 dyn_hair_1_r [{
   shapeType=Capsule
   length=3.5
   radius=2.5
   shapeName=dyn_hair_1_r_capsule
   center=<-3.8657 0.9355 -0.0004>
   rotation=<rot 90 0.0002 -179.9996>
  }]
 dyn_hair_3_l [{shapeType=Capsule, length=3.5, radius=2, shapeName=dyn_hair_3_l_capsule, center=<1.9557 -0.9068 -0.0001>, rotation=<rot 90 0 -179.9998>}]
 dyn_hair_3_r [{shapeType=Capsule, length=0.75, radius=2.5, shapeName=dyn_hair_3_r_capsule, center=<-2.6137 0 -0.0003>, rotation=<rot 90 0.0002 -179.9996>}]
 dyn_hair_bk_1 [{shapeType=Capsule, length=5.5, radius=2, shapeName=dyn_hair_bk_1_capsule, center=<3.3532 -1.0863 0.0004>, rotation=<rot 90 0 0>}]
 dyn_hair_bk_1_l [{shapeType=Capsule, length=2.5, radius=2.5, shapeName=dyn_hair_bk_1_l_capsule, center=<3.3532 -0.8596 0.0003>, rotation=<rot 90 0 0>}]
 dyn_hair_bk_1_r [{shapeType=Capsule, length=4.5, radius=2.5, shapeName=dyn_hair_bk_1_r_capsule, center=<-3.5939 0.4677 0.0004>, rotation=<rot 90 0 -11.8007>}]
 dyn_hair_bk_2 [{shapeType=Capsule, length=2.75, radius=2, shapeName=dyn_hair_bk_2_capsule, center=<4.5822 0 0.0003>, rotation=<rot 90 0 0>}]
 dyn_hair_bk_2_l [{shapeType=Capsule, length=2.25, radius=2.25, shapeName=dyn_hair_bk_2_l_capsule, center=<4.2839 0 0.0003>, rotation=<rot 90 0 0>}]
 dyn_hair_bk_2_r [{shapeType=Capsule, length=2.25, radius=2.5, shapeName=dyn_hair_bk_2_r_capsule, center=<-4.3018 -0.1694 0.0004>, rotation=<rot 90 0 -4.8541>}]
 dyn_hair_bk_3 [{shapeType=Capsule, length=3.092, radius=1.8023, shapeName=dyn_hair_bk_3_capsule, center=<3.8065 -1.9933 0>, rotation=<rot 90 0 18.7642>}]
 dyn_hair_bk_3_l [{shapeType=Capsule, length=3.5354, radius=2.0715, shapeName=dyn_hair_bk_3_l_capsule, center=<3.6399 -1.8581 0>, rotation=<rot 90 0 0>}]
 dyn_hair_bk_3_r [{
   shapeType=Capsule
   length=2.3671
   radius=2.25
   shapeName=dyn_hair_bk_3_r_capsule
   center=<-4.6963 1.3539 0.9925>
   rotation=<rot 80.3414 -16.6359 0.0001>
  }]
 dyn_hair_frt_1_l [{shapeType=Capsule, length=4.25, radius=1.25, shapeName=dyn_hair_frt_1_l_capsule, center=<2.4572 0.0002 0>, rotation=<rot 90 0 179.9996>}]
 dyn_hair_frt_1_r [{shapeType=Capsule, length=4.25, radius=1.25, shapeName=dyn_hair_frt_1_r_capsule, center=<-2.6137 0 -0.0003>, rotation=<rot 90 0 179.9996>}]
 dyn_hair_frt_2_l [{
   shapeType=Capsule
   length=4.25
   radius=1.25
   shapeName=dyn_hair_frt_2_l_capsule
   center=<2.4572 0.0002 0>
   rotation=<rot 90 0.0003 -179.9998>
  }]
 dyn_hair_frt_2_r [{shapeType=Capsule, length=2.75, radius=1.25, shapeName=dyn_hair_frt_2_r_capsule, center=<-2.6137 0 -0.0003>, rotation=<rot 90 0 179.9996>}]
 dyn_hair_frt_3_l [{shapeType=Capsule, length=4.25, radius=1.25, shapeName=dyn_hair_frt_3_l_capsule, center=<2.4572 0.0002 0>, rotation=<rot 90 0.0001 179.9999>}]
 dyn_hair_frt_3_r [{
   shapeType=Capsule
   length=4.25
   radius=1.25
   shapeName=dyn_hair_frt_3_r_capsule
   center=<-2.6137 0 -0.0003>
   rotation=<rot 90 -0.0002 179.9996>
  }]
 dyn_hair_frt_4_l [{shapeType=Capsule, length=4.25, radius=1.25, shapeName=dyn_hair_frt_4_l_capsule, center=<2.4572 0.0002 0>, rotation=<rot 90 0.0001 179.9999>}]
 dyn_hair_frt_4_r [{
   shapeType=Capsule
   length=4.25
   radius=1.25
   shapeName=dyn_hair_frt_4_r_capsule
   center=<-2.6137 0 -0.0003>
   rotation=<rot 90 -0.0002 179.9996>
  }]
 head [{shapeType=Capsule, length=4.7802, radius=5.8823, center=<5.5002 -2.261 0.0716>, rotation=<rot 0 90 90>}]
 neck_01 [
  {shapeType=Capsule, length=5, radius=6.25, shapeName=neck_01_capsule, center=<-3.5404 0.0972 -0.0001>, rotation=<rot 0 -90 0>}
  {shapeType=Capsule, length=5, radius=5.25, shapeName=neck_01_capsule, center=<2.2298 1.0938 -0.0001>, rotation=<rot 90 0 0>}
  {shapeType=Capsule, length=7, radius=5.75, shapeName=neck_01_capsule, center=<4.4655 -2.9697 -0.0001>, rotation=<rot 90 172.875 -178.3008>}
 ]
 spine_05 [
  {shapeType=Capsule, length=23, radius=8.5, shapeName=spine_05_capsule, center=<1.8986 -0.099 0>, rotation=<rot 0 -90 0>}
  {shapeType=Capsule, length=31.75, radius=1.75, shapeName=spine_05_capsule, center=<12.2715 -3.831 0.0004>}
  {shapeType=Capsule, length=4.75, radius=7, shapeName=spine_05_capsule, center=<7.9718 -0.4417 -0.0001>, rotation=<rot 0 0 90>}
  {shapeType=Capsule, length=6, radius=7, shapeName=spine_05_capsule, center=<7.9775 -0.4624 0>}
  {shapeType=Capsule, length=15, radius=8, shapeName=spine_05_capsule, center=<4.3274 -4.0926 0.0003>}
  {shapeType=Capsule, length=17.75, radius=12.25, shapeName=spine_05_capsule, center=<-4.8774 0.1025 0>}
 ]
}
constraints [
 {bone1Name=dyn_hair_1_l, bone2Name=head, swing1LimitDegrees=30, swing2LimitDegrees=45, twistLimitDegrees=10}
 {bone1Name=dyn_hair_3_l, bone2Name=dyn_hair_1_l, swing1LimitDegrees=30, swing2LimitDegrees=30, twistLimitDegrees=10}
 {bone1Name=dyn_hair_bk_1_l, bone2Name=head, swing1LimitDegrees=30, swing2LimitDegrees=30, twistLimitDegrees=10}
 {bone1Name=dyn_hair_bk_2_l, bone2Name=dyn_hair_bk_1_l, swing1LimitDegrees=20, swing2LimitDegrees=20, twistLimitDegrees=10}
 {bone1Name=dyn_hair_bk_1, bone2Name=head, swing1LimitDegrees=90, swing2LimitDegrees=15, twistLimitDegrees=0}
 {bone1Name=dyn_hair_bk_2, bone2Name=dyn_hair_bk_1, swing1LimitDegrees=45, swing2LimitDegrees=10, twistLimitDegrees=10}
 {bone1Name=dyn_hair_frt_1_l, bone2Name=head, swing1LimitDegrees=20, swing2LimitDegrees=20, twistLimitDegrees=5}
 {bone1Name=dyn_hair_frt_2_l, bone2Name=dyn_hair_frt_1_l, swing1LimitDegrees=20, swing2LimitDegrees=20, twistLimitDegrees=5}
 {bone1Name=dyn_hair_frt_3_l, bone2Name=dyn_hair_frt_2_l, swing1LimitDegrees=20, swing2LimitDegrees=20, twistLimitDegrees=5}
 {bone1Name=dyn_hair_frt_4_l, bone2Name=dyn_hair_frt_3_l, swing1LimitDegrees=20, swing2LimitDegrees=20, twistLimitDegrees=5}
 {bone1Name=dyn_hair_1_r, bone2Name=head, swing1LimitDegrees=23.1429, swing2LimitDegrees=50.5714, twistLimitDegrees=10}
 {bone1Name=dyn_hair_3_r, bone2Name=dyn_hair_1_r, swing1LimitDegrees=30, swing2LimitDegrees=30, twistLimitDegrees=10}
 {bone1Name=dyn_hair_bk_1_r, bone2Name=head, swing1LimitDegrees=70, swing2LimitDegrees=10, twistLimitDegrees=0}
 {bone1Name=dyn_hair_bk_2_r, bone2Name=dyn_hair_bk_1_r, swing1LimitDegrees=50, swing2LimitDegrees=30, twistLimitDegrees=10}
 {bone1Name=dyn_hair_frt_1_r, bone2Name=head, swing1LimitDegrees=20, swing2LimitDegrees=20, twistLimitDegrees=5}
 {bone1Name=dyn_hair_frt_2_r, bone2Name=dyn_hair_frt_1_r, swing1LimitDegrees=20, swing2LimitDegrees=20, twistLimitDegrees=5}
 {bone1Name=dyn_hair_frt_3_r, bone2Name=dyn_hair_frt_2_r, swing1LimitDegrees=20, swing2LimitDegrees=20, twistLimitDegrees=5}
 {bone1Name=dyn_hair_frt_4_r, bone2Name=dyn_hair_frt_3_r, swing1LimitDegrees=20, swing2LimitDegrees=20, twistLimitDegrees=5}
 {
  bone1Name=head
  bone2Name=neck_01
  swing1LimitDegrees=45
  swing1Motion=Free
  swing2LimitDegrees=45
  swing2Motion=Free
  twistLimitDegrees=45
  twistMotion=Free
 }
 {
  bone1Name=neck_01
  bone2Name=spine_05
  swing1LimitDegrees=45
  swing1Motion=Free
  swing2LimitDegrees=45
  swing2Motion=Free
  twistLimitDegrees=45
  twistMotion=Free
 }
 {bone1Name=dyn_hair_bk_3_l, bone2Name=dyn_hair_bk_2_l, swing1LimitDegrees=10, swing2LimitDegrees=10, twistLimitDegrees=0}
 {bone1Name=dyn_hair_bk_3, bone2Name=dyn_hair_bk_2, swing1LimitDegrees=20, swing2LimitDegrees=5, twistLimitDegrees=0}
 {bone1Name=dyn_hair_bk_3_r, bone2Name=dyn_hair_bk_2_r, swing1LimitDegrees=50, swing2LimitDegrees=10, twistLimitDegrees=0}
 {
  bone1Name=clavicle_l
  bone2Name=spine_05
  swing1LimitDegrees=45
  swing1Motion=Free
  swing2LimitDegrees=45
  swing2Motion=Free
  twistLimitDegrees=45
  twistMotion=Free
 }
 {
  bone1Name=clavicle_r
  bone2Name=spine_05
  swing1LimitDegrees=45
  swing1Motion=Free
  swing2LimitDegrees=45
  swing2Motion=Free
  twistLimitDegrees=45
  twistMotion=Free
 }
]
