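# Anim Blueprint for the F_MED_Critter_Streak face accessory, adding hair dynamics to the pose copied from its attached parent.
# After a fixed hair-bone rotation, front-end and gameplay RigidBody nodes with separate physics assets are toggled by front-end state.
# The gameplay simulation takes gravity from a deprecated state-based override function.
# Head-to-neck yaw, range-mapped to alpha, then drives two additive translations on dyn_hair_bk_1.
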AnimBlueprint /BRCosmetics/Characters/Player/Female/Medium/Bodies/F_MED_Critter_Streak/Meshes/Parts/F_MED_Critter_Streak_FaceAcc_AnimBP
target_skeleton=&/F_MED_Critter_Streak_FaceAcc_Skeleton
graphs [
 AnimGraph {nodes [
   n1 = "Transform (Modify) Bone - Bone: dyn_hair_bk_1_r"(
    ComponentPose=Local To Component(LocalPose=Copy Pose From Mesh() {bUseAttachedParent=true})
    Alpha=1
    Translation="0,0,0"
    Rotation="0,-60,0"
    Scale="1,1,1"
   ) {rotationMode=BMM_Additive, alphaBoolBlend {blendOption=Linear}, boneToModify {boneName=dyn_hair_bk_1_r}, rotation {yaw=-60}}
   n2 = RigidBody!(ComponentPose=@n1, Alpha=Get Is in Front End() -> (Is in Front End:double), OverrideWorldGravity:Vector="0,0,-980", ExternalForce) -> (Pose) {
    bClampLinearTranslationLimitToRefPose=true
    bOverrideWorldGravity=true
    overridePhysicsAsset=&/F_MED_Critter_Streak_FaceAcc_Physics_FrontEnd
    simulationSpace=BaseBoneSpace
    alphaBoolBlend {blendOption=Linear}
    baseBoneRef {boneName=root}
    componentAppliedLinearAccClamp=<1500 1500 1500>
    componentLinearAccScale=<0.01 0.01 0.01>
    componentLinearVelScale=<0.01 0.01 0.01>
    overrideWorldGravity {z=-980}
   }
   n3 = RigidBody(ComponentPose=@n2, Alpha=Get Is not Front End() -> (Is not Front End:double), OverrideWorldGravity=Get Gravity Override()) {
    bClampLinearTranslationLimitToRefPose=true
    bOverrideWorldGravity=true
    overridePhysicsAsset=&/F_MED_Critter_Streak_FaceAcc_Physics
    simulationSpace=BaseBoneSpace
    alphaBoolBlend {blendOption=Linear}
    baseBoneRef {boneName=root}
    componentAppliedLinearAccClamp=<1500 1500 1500>
    componentLinearAccScale=<0.01 0.01 0.01>
    componentLinearVelScale=<0.01 0.01 0.01>
    overrideWorldGravity {z=-980}
   }
   n4 = Get head_fwdBwd_rot() -> (head_fwdBwd_rot:double)
   n5 = "Transform (Modify) Bone - Bone: dyn_hair_bk_1"(ComponentPose=@n3, Alpha=@n4, Translation="1,-2,0", Rotation="0,0,0", Scale="1,1,1") {
    rotationMode=BMM_Additive
    translationMode=BMM_Additive
    translationSpace=BCS_BoneSpace
    alphaBoolBlend {blendOption=Linear}
    alphaScaleBiasClamp {bMapRange=true, inRange {max=5, min=-50}, outRange {max=0, min=1}}
    boneToModify {boneName=dyn_hair_bk_1}
    translation=<1 -2>
   }
   n6 = "Transform (Modify) Bone - Bone: dyn_hair_bk_1"(ComponentPose=@n5, Alpha=@n4, Translation="2,-3,0", Rotation="0,0,0", Scale="1,1,1") {
    rotationMode=BMM_Additive
    translationMode=BMM_Additive
    translationSpace=BCS_BoneSpace
    alphaBoolBlend {blendOption=Linear}
    alphaScaleBiasClamp {bMapRange=true, inRange {max=5, min=-50}, outRange {max=0, min=1}}
    boneToModify {boneName=dyn_hair_bk_1}
    translation=<2 -3>
   }
   Output Pose(Result=Component To Local(ComponentPose=@n6)) {layerGroup=None}
  ]}
 EventGraph {nodes [
   Event BlueprintUpdateAnimation()
   > n1 = Sequence!(execute) -> (then_0, then_1, then_2:Exec, then_3:Exec, then_4:Exec)
   > Set Is Sky Diving alpha(Is Sky Diving alpha:double="To Float (Boolean)"(InBool=Get bIsSkyDiving())) -> (Output_Get:double)
   n2 = Gravity Override State Function DEPRECATED(
    execute=@n1.then_1
    CharacterPartAnimInstance=Self-Reference()
    GameTimeInSeconds=GetGameTimeInSeconds()
    GravityOverrideParams=Get Gravity Override Params()
   )
   > Set Gravity Override(Gravity Override=@n2)
   n3 = GetDeltaTransformFromRefPose(self=GetOwnerSkeletalMeshComponent(), BoneName=head, BaseName=neck_01)
   Set head_fwdBwd_rot(execute=@n1.then_2, head_fwdBwd_rot:double=BreakRotator(InRot=BreakTransform(InTransform=@n3).Rotation).Yaw) -> (Output_Get:double)
   n4 = Branch(execute=@n1.then_3, Condition=Get bIsInFrontEnd())
   > Set Is in Front End(Is in Front End:double=1) -> (Output_Get:double)
   > Set Is not Front End(Is not Front End:double=0) -> (Output_Get:double)
   Set Is in Front End(execute=@n4.else, Is in Front End=0)
   > Set Is not Front End(Is not Front End=1)
  ]}
]
variables [
 Character Part Anim Instance {type=CustomCharacterPartAnimInstance, default=null}
 Gravity Override {type=Vector}
 Gravity Override Params {
  type=GravityOverrideParamsStruct
  default {skydive_FloatingDown {gravityOverride=<0 -700 400>, windAmplitude=<200 -500 100>, windOffset=<0 -600 100>}}
 }
 head_fwdBwd_rot {type=number}
 Is in Front End {type=number}
 Is not Front End {type=number}
 Is Sky Diving alpha {type=number}
 NewVar_0 {
  type=BodyInstance
  default {
   angularDamping=0
   bAllowPartialIslandSleep=true
   bAutoWeld=false
   bEnableGravity=true
   bGenerateWakeEvents=false
   bGyroscopicTorqueEnabled=false
   bIgnoreAnalyticCollisions=false
   bInertiaConditioning=true
   bInterpolateWhenSubStepping=true
   bLockRotation=true
   bLockTranslation=true
   bLockXRotation=false
   bLockXTranslation=false
   bLockYRotation=false
   bLockYTranslation=false
   bLockZRotation=false
   bLockZTranslation=false
   bNotifyRigidBodyCollision=false
   bOneWayInteraction=false
   bOverrideIterationCounts=false
   bOverrideMass=false
   bOverrideMaxAngularVelocity=false
   bOverrideMaxDepenetrationVelocity=false
   bOverrideSolverAsyncDeltaTime=false
   bOverrideWalkableSlopeOnInstance=false
   bSimulatePhysics=false
   bSmoothEdgeCollisions=false
   bStartAwake=true
   bUpdateKinematicFromSimulation=false
   bUpdateMassWhenScaleChanges=true
   bUseCCD=false
   bUseMACD=false
   collisionEnabled=QueryAndPhysics
   collisionProfileName=Custom
   customSleepThresholdMultiplier=1
   dOFMode=Default
   gravityGroupIndex=0
   linearDamping=0.01
   massInKgOverride=100
   massScale=1
   maxAngularVelocity=3600
   maxDepenetrationVelocity=0
   objectType=ECC_WorldStatic
   physMaterialOverride=null
   physicsBlendWeight=0
   positionSolverIterationCount=8
   projectionSolverIterationCount=1
   sleepFamily=Normal
   solverAsyncDeltaTime=0.0167
   stabilizationThresholdMultiplier=1
   velocitySolverIterationCount=1
   cOMNudge=<0 0 0>
   collisionResponses {responseArray []}
   customDOFPlaneNormal=<0 0 0>
   inertiaTensorScale=<1 1 1>
   walkableSlopeOverride {walkableSlopeAngle=0, walkableSlopeBehavior=WalkableSlope_Default}
  }
 }
 NewVar_1 {type=SkeletalMeshComponent}
]
